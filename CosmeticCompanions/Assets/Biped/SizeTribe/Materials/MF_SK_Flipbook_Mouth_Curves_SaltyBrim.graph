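# Material function that picks a mouth flipbook frame index for the SaltyBrim cosmetic companion from animation-driven facial curves.
# Scalar pose parameters such as smile, frown, jaw, lips and emotes are paired with tunable threshold parameters, and named constants define the frame indices.
# A nested If chain selects a single index by priority, with Min and Max comparisons apparently testing combined poses against their thresholds.
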MaterialFunction /CosmeticCompanions/Assets/Biped/SizeTribe/Materials/MF_SK_Flipbook_Mouth_Curves_SaltyBrim
graph {nodes [
  FunctionOutput(None=NamedRerouteUsage()) {outputName=Index}
  NamedRerouteDeclaration(None=ScalarParameter() {group="Flipbook 04.01 Mouth | Curves (Driven By Animation)", parameterName=R_frown_pose, sortPriority=1}) {name=Frown_Pose, nodeColor=<rgba 1 0.2105 0 1>}
  NamedRerouteDeclaration(None=ScalarParameter() {defaultValue=0.1, group=Flipbook 04.02 Mouth | Thresholds, parameterName=Threshold_Frown, sortPriority=1}) {name=Threshold_Frown, nodeColor=<rgba 1 0.0834 0.4399 1>}
  NamedRerouteDeclaration(None=Constant() {r=3}) {name=Index_Mouth_Frown_Open, nodeColor=<rgba 0.2818 0.3488 1 1>}
  NamedRerouteDeclaration(None=ScalarParameter() {group="Flipbook 04.01 Mouth | Curves (Driven By Animation)", parameterName=R_smile_pose, sortPriority=0}) {name=Smile_Pose, nodeColor=<rgba 1 0.2105 0 1>}
  NamedRerouteDeclaration(None=ScalarParameter() {group="Flipbook 04.01 Mouth | Curves (Driven By Animation)", parameterName=jaw_open_pose, sortPriority=2}) {name=Jaw_Open_Pose, nodeColor=<rgba 1 0.2105 0 1>}
  NamedRerouteDeclaration(None=Constant() {r=8}) {name=Index_Mouth_Angry_Closed, nodeColor=<rgba 0.2818 0.3488 1 1>}
  NamedRerouteDeclaration(None=Constant()) {name=Index_Neutral, nodeColor=<rgba 0.2818 0.3488 1 1>}
  NamedRerouteDeclaration(None=Constant() {r=5}) {name=Index_Mouth_Press, nodeColor=<rgba 0.2818 0.3488 1 1>}
  NamedRerouteDeclaration(None=Constant() {r=2}) {name=Index_Mouth_Neutral_Open, nodeColor=<rgba 0.2818 0.3488 1 1>}
  NamedRerouteDeclaration(None=Constant() {r=12}) {name=Index_Mouth_Emote_01, nodeColor=<rgba 0.2818 0.3488 1 1>}
  NamedRerouteDeclaration(None=Constant() {r=10}) {name=Index_Mouth_Emote_Ee, nodeColor=<rgba 0.2818 0.3488 1 1>}
  NamedRerouteDeclaration(None=Constant() {r=11}) {name=Index_Mouth_Emote_Oo, nodeColor=<rgba 0.2818 0.3488 1 1>}
  NamedRerouteDeclaration(None=ScalarParameter() {group="Flipbook 04.01 Mouth | Curves (Driven By Animation)", parameterName=reverse_jaw_pose, sortPriority=3}) {name=Reverse_Jaw_Pose, nodeColor=<rgba 1 0.2105 0 1>}
  NamedRerouteDeclaration(None=ScalarParameter() {group="Flipbook 04.01 Mouth | Curves (Driven By Animation)", parameterName=R_lip_corner_wide_pose, sortPriority=4}) {name=R_Lip_Corner_Wide_Pose, nodeColor=<rgba 1 0.2105 0 1>}
  NamedRerouteDeclaration(None=ScalarParameter() {group="Flipbook 04.01 Mouth | Curves (Driven By Animation)", parameterName=jaw_up_pose, sortPriority=5}) {name=Jaw_Up_Pose, nodeColor=<rgba 1 0.2105 0 1>}
  NamedRerouteDeclaration(None=Constant() {r=4}) {name=Index_Mouth_Neutral_Wide, nodeColor=<rgba 0.2818 0.3488 1 1>}
  NamedRerouteDeclaration(None=ScalarParameter() {group="Flipbook 04.01 Mouth | Curves (Driven By Animation)", parameterName=R_lower_lip_down_pose, sortPriority=7}) {name=R_Lower_Lip_Down_Pose, nodeColor=<rgba 1 0.2105 0 1>}
  NamedRerouteDeclaration(None=ScalarParameter() {group="Flipbook 04.01 Mouth | Curves (Driven By Animation)", parameterName=R_upper_lip_raiser_pose, sortPriority=8}) {name=R_Upper_Lip_Raiser_Pose, nodeColor=<rgba 1 0.2105 0 1>}
  NamedRerouteDeclaration(None=ScalarParameter() {group="Flipbook 04.01 Mouth | Curves (Driven By Animation)", parameterName=mouth_emote_01, sortPriority=12}) {name=Mouth_Emote_01_Pose, nodeColor=<rgba 1 0.2105 0 1>}
  NamedRerouteDeclaration(None=ScalarParameter() {group="Flipbook 04.01 Mouth | Curves (Driven By Animation)", parameterName=mouth_emote_Ee, sortPriority=10}) {name=Mouth_Emote_Ee_Pose, nodeColor=<rgba 1 0.2105 0 1>}
  NamedRerouteDeclaration(None=If(A=NamedRerouteUsage(), B=NamedRerouteUsage(), "A > B"=NamedRerouteUsage(), "A < B"=NamedRerouteUsage())) {name=Smile_Output, nodeColor=<rgba 0 0.1532 1 1>}
  NamedRerouteDeclaration(None=ScalarParameter() {group="Flipbook 04.01 Mouth | Curves (Driven By Animation)", parameterName=mouth_emote_Oo, sortPriority=11}) {name=Mouth_Emote_Oo_Pose, nodeColor=<rgba 1 0.2105 0 1>}
  NamedRerouteDeclaration(None=ScalarParameter() {defaultValue=0.21, group=Flipbook 04.02 Mouth | Thresholds, parameterName=Threshold_Reverse_Jaw, sortPriority=3}) {name=Threshold_Reverse_Jaw, nodeColor=<rgba 1 0.0834 0.4399 1>}
  NamedRerouteDeclaration(None=ScalarParameter() {defaultValue=0.1, group=Flipbook 04.02 Mouth | Thresholds, parameterName=Threshold_R_Lip_Corner_Wide, sortPriority=4}) {name=Threshold_R_Lip_Corner_Wide, nodeColor=<rgba 1 0.0834 0.4399 1>}
  NamedRerouteDeclaration(None=ScalarParameter() {defaultValue=0.3, group=Flipbook 04.02 Mouth | Thresholds, parameterName=Threshold_Jaw_Up, sortPriority=5}) {name=Threshold_Jaw_Up, nodeColor=<rgba 1 0.0834 0.4399 1>}
  NamedRerouteDeclaration(None=ScalarParameter() {defaultValue=0.1, group=Flipbook 04.02 Mouth | Thresholds, parameterName=Threshold_Lower_Lip_Down, sortPriority=6}) {name=Threshold_Lower_Lip_Down, nodeColor=<rgba 1 0.0834 0.4399 1>}
  NamedRerouteDeclaration(None=ScalarParameter() {defaultValue=0.1, group=Flipbook 04.02 Mouth | Thresholds, parameterName=Threshold_Upper_Lip_Raiser, sortPriority=7}) {name=Threshold_Upper_Lip_Raiser, nodeColor=<rgba 1 0.0834 0.4399 1>}
  NamedRerouteDeclaration(None=If(A=NamedRerouteUsage(), B=NamedRerouteUsage(), "A > B"=NamedRerouteUsage(), "A < B"=NamedRerouteUsage())) {name=Frown_Output, nodeColor=<rgba 0 0.1532 1 1>}
  NamedRerouteDeclaration(None=ScalarParameter() {defaultValue=0.1, group=Flipbook 04.02 Mouth | Thresholds, parameterName=Threshold_Mouth_Emote_Oo, sortPriority=11}) {name=Threshold_Mouth_Emote_Oo, nodeColor=<rgba 1 0.0834 0.4399 1>}
  NamedRerouteDeclaration(None=ScalarParameter() {defaultValue=0.1, group=Flipbook 04.02 Mouth | Thresholds, parameterName=Threshold_Mouth_Emote_01, sortPriority=12}) {name=Threshold_Mouth_Emote_01, nodeColor=<rgba 1 0.0834 0.4399 1>}
  NamedRerouteDeclaration(None=ScalarParameter() {defaultValue=0.1, group=Flipbook 04.02 Mouth | Thresholds, parameterName=Threshold_Mouth_Emote_Ee, sortPriority=10}) {name=Threshold_Mouth_Emote_Ee, nodeColor=<rgba 1 0.0834 0.4399 1>}
  NamedRerouteDeclaration(None=Constant() {r=9}) {name=Index_Mouth_Angry_Open, nodeColor=<rgba 0.2818 0.3488 1 1>}
  NamedRerouteDeclaration(None=If(A=NamedRerouteUsage(), B=NamedRerouteUsage(), "A > B"=NamedRerouteUsage(), "A < B"=NamedRerouteUsage())) {name=Jaw_Open_Output, nodeColor=<rgba 0 0.1532 1 1>}
  NamedRerouteDeclaration(None=Constant() {r=6}) {name=Index_Mouth_Smile_Closed, nodeColor=<rgba 0.2818 0.3488 1 1>}
  NamedRerouteDeclaration(None=Constant() {r=7}) {name=Index_Mouth_Smile_Open, nodeColor=<rgba 0.2818 0.3488 1 1>}
  NamedRerouteDeclaration(None=Constant() {r=1}) {name=Index_Mouth_Frown_Closed, nodeColor=<rgba 0.2818 0.3488 1 1>}
  n1 = NamedRerouteUsage()
  n2 = NamedRerouteUsage()
  n3 = NamedRerouteUsage()
  n4 = Min(A=NamedRerouteUsage(), B=NamedRerouteUsage())
  n5 = NamedRerouteUsage()
  n6 = Max(A=NamedRerouteUsage(), B=NamedRerouteUsage())
  n7 = NamedRerouteUsage()
  n8 = Min(A=NamedRerouteUsage(), B=NamedRerouteUsage())
  n9 = Max(A=NamedRerouteUsage(), B=NamedRerouteUsage())
  n10 = NamedRerouteUsage()
  n11 = Min(A=NamedRerouteUsage(), B=NamedRerouteUsage())
  n12 = Max(A=NamedRerouteUsage(), B=NamedRerouteUsage())
  n13 = NamedRerouteUsage()
  n14 = NamedRerouteUsage()
  n15 = NamedRerouteUsage()
  n16 = NamedRerouteUsage()
  n17 = If(A=NamedRerouteUsage(), B=NamedRerouteUsage(), "A > B"=NamedRerouteUsage(), "A < B"=NamedRerouteUsage())
  n18 = If(
   A=Min(A=NamedRerouteUsage(), B=NamedRerouteUsage())
   B=Max(A=NamedRerouteUsage(), B=NamedRerouteUsage())
   "A > B"=NamedRerouteUsage()
   "A < B"=If(A=NamedRerouteUsage(), B=NamedRerouteUsage(), "A > B"=NamedRerouteUsage(), "A < B"=If(A=@n14, B=@n15, "A > B"=@n16, "A < B"=@n17)) {constB=1}
  )
  n19 = If(
   A=Min(A=NamedRerouteUsage(), B=NamedRerouteUsage())
   B=Max(A=NamedRerouteUsage(), B=NamedRerouteUsage())
   "A > B"=NamedRerouteUsage()
   "A < B"=If(A=NamedRerouteUsage(), B=NamedRerouteUsage(), "A > B"=NamedRerouteUsage(), "A < B"=If(A=@n11, B=@n12, "A > B"=@n13, "A < B"=@n18))
  )
  n20 = If(
   A=NamedRerouteUsage()
   B=NamedRerouteUsage()
   "A > B"=NamedRerouteUsage()
   "A < B"=If(A=Min(A=@n4, B=@n5), B=Max(A=@n6, B=@n7), "A > B"=NamedRerouteUsage(), "A < B"=If(A=@n8, B=@n9, "A > B"=@n10, "A < B"=@n19))
  )
  NamedRerouteDeclaration(None=If(A=NamedRerouteUsage(), B=NamedRerouteUsage(), "A > B"=NamedRerouteUsage(), "A < B"=If(A=@n1, B=@n2, "A > B"=@n3, "A < B"=@n20))) {name=Final Mouth Output, nodeColor=<rgba 1 0.0234 0 1>}
  NamedRerouteDeclaration(None=ScalarParameter() {defaultValue=0.1, group=Flipbook 04.02 Mouth | Thresholds, parameterName=Threshold_Smile, sortPriority=0}) {name=Threshold_Smile, nodeColor=<rgba 1 0.0834 0.4399 1>}
  NamedRerouteDeclaration(None=ScalarParameter() {defaultValue=0.1, group=Flipbook 04.02 Mouth | Thresholds, parameterName=Threshold_Jaw_Open, sortPriority=2}) {name=Threshold_Jaw_Open, nodeColor=<rgba 1 0.0834 0.4399 1>}
 ]}
