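# Post-process Animation Blueprint for the SizeTribe biped companion, adding secondary motion to both tails, hair and whiskers.
# The input pose, or a copy with additive tail-bone rotations, is selected by a bool blend, cached, then run through chained spring-based Anim Dynamics nodes to the output.
# The event graph derives tail values from animation curves and flags companion or mannequin owners, with no visible links to the AnimGraph.
# The PoseAsset blend node is unconnected.
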AnimBlueprint /CosmeticCompanions/Assets/Biped/SizeTribe/Meshes/SizeTribe_Biped_Post_AnimBP
target_skeleton=&/SizeTribe_Biped_Skeleton
graphs [
 AnimGraph {nodes [
   "Transform (Modify) Bone - Bone: tail_01_r"!(ComponentPose, Alpha=1, Rotation="-25,-2.5,0") -> (Pose) {
    rotationMode=BMM_Additive
    rotationSpace=BCS_BoneSpace
    alphaScaleBiasClamp {bInterpResult=true, interpSpeedDecreasing=8}
    boneToModify {boneName=tail_01_r}
    rotation {pitch=-25, yaw=-2.5}
   }
   Pose!(PoseAsset:Pose Asset Object Reference=Get PoseAsset() -> (PoseAsset:Pose Asset Object Reference), SourcePose:Pose Link Structure) -> (Pose:Pose Link Structure) {is=title AnimGraphNode_PoseBlendNode, bIgnorePoseAssetSourceAnimationCurves=true, blendOption=Linear}
   Output Pose(Result=Use cached pose 'DynamicsCache'())
   Save cached pose 'CRCache_Input'(Pose=Input Pose()) {cacheName=CRCache_Input}
   n1 = Anim Dynamics - Bone: tail_01!(ComponentPose=Local To Component(LocalPose=Use cached pose 'preDynamicsCache'()), Alpha=1, GravityOverride:Vector="0,0,-50") -> (Pose) {
    angularDampingOverride=1
    angularSpringConstant=7500
    bAngularSpring=true
    bOverrideAngularDamping=true
    bUseGravityOverride=true
    gravityScale=0.35
    simulationSpace=RootRelative
    boundBone {boneName=tail_01}
    componentAppliedLinearAccClamp=<1000 1000 1000>
    componentLinearAccScale=<1 1 1>
    componentLinearVelScale=<1 1 1>
    gravityOverride {z=-100}
    physicsBodyDefinitions [{
      boundBone {boneName=tail_01}
      boxExtents=<25 35 35>
      constraintSetup {angularLimitsMax=<0 25 25>, angularLimitsMin=<0 -25 -25>, angularTarget=<1 0 0>}
      localJointOffset=<3 0 0>
     }]
    simSpaceSettings {maxAngularAcceleration=25, maxAngularVelocity=5, simSpaceAngularAlpha=1}
   }
   n2 = Anim Dynamics - Bone: tail_02!(ComponentPose=@n1, Alpha=1, GravityOverride:Vector="0,0,-50") -> (Pose) {
    angularDampingOverride=1
    angularSpringConstant=7500
    bAngularSpring=true
    bOverrideAngularDamping=true
    bUseGravityOverride=true
    gravityScale=0.35
    simulationSpace=RootRelative
    boundBone {boneName=tail_02}
    componentAppliedLinearAccClamp=<1000 1000 1000>
    componentLinearAccScale=<1 1 1>
    componentLinearVelScale=<1 1 1>
    gravityOverride {z=-100}
    physicsBodyDefinitions [{
      boundBone {boneName=tail_02}
      boxExtents=<25 35 35>
      constraintSetup {angularLimitsMax=<0 25 25>, angularLimitsMin=<0 -25 -25>, angularTarget=<1 0 0>}
      localJointOffset=<3 0 0>
     }]
    simSpaceSettings {maxAngularAcceleration=25, maxAngularVelocity=5, simSpaceAngularAlpha=1}
   }
   n3 = Anim Dynamics - Bone: tail_03!(ComponentPose=@n2, Alpha=1, GravityOverride:Vector="0,0,-50") -> (Pose) {
    angularDampingOverride=1
    angularSpringConstant=7500
    bAngularSpring=true
    bOverrideAngularDamping=true
    bUseGravityOverride=true
    gravityScale=0.35
    simulationSpace=RootRelative
    boundBone {boneName=tail_03}
    componentAppliedLinearAccClamp=<1000 1000 1000>
    componentLinearAccScale=<1 1 1>
    componentLinearVelScale=<1 1 1>
    gravityOverride {z=-100}
    physicsBodyDefinitions [{
      boundBone {boneName=tail_03}
      boxExtents=<25 35 35>
      constraintSetup {angularLimitsMax=<0 25 25>, angularLimitsMin=<0 -25 -25>, angularTarget=<1 0 0>}
      localJointOffset=<3 0 0>
     }]
    simSpaceSettings {maxAngularAcceleration=25, maxAngularVelocity=5, simSpaceAngularAlpha=1}
   }
   n4 = Anim Dynamics - Bone: tail_04!(ComponentPose=@n3, Alpha=1, GravityOverride:Vector="0,0,-50") -> (Pose) {
    angularDampingOverride=1
    angularSpringConstant=7500
    bAngularSpring=true
    bOverrideAngularDamping=true
    bUseGravityOverride=true
    gravityScale=0.35
    simulationSpace=RootRelative
    boundBone {boneName=tail_04}
    componentAppliedLinearAccClamp=<1000 1000 1000>
    componentLinearAccScale=<1 1 1>
    componentLinearVelScale=<1 1 1>
    gravityOverride {z=-100}
    physicsBodyDefinitions [{
      boundBone {boneName=tail_04}
      boxExtents=<25 35 35>
      constraintSetup {angularLimitsMax=<0 25 25>, angularLimitsMin=<0 -25 -25>, angularTarget=<1 0 0>}
      localJointOffset=<3 0 0>
     }]
    simSpaceSettings {maxAngularAcceleration=25, maxAngularVelocity=5, simSpaceAngularAlpha=1}
   }
   n5 = Anim Dynamics - Bone: tail_05!(ComponentPose=@n4, Alpha=1, GravityOverride:Vector="0,0,-50") -> (Pose) {
    angularDampingOverride=1
    angularSpringConstant=7500
    bAngularSpring=true
    bOverrideAngularDamping=true
    bUseGravityOverride=true
    gravityScale=0.35
    simulationSpace=RootRelative
    boundBone {boneName=tail_05}
    componentAppliedLinearAccClamp=<1000 1000 1000>
    componentLinearAccScale=<1 1 1>
    componentLinearVelScale=<1 1 1>
    gravityOverride {z=-100}
    physicsBodyDefinitions [{
      boundBone {boneName=tail_05}
      boxExtents=<25 35 35>
      constraintSetup {angularLimitsMax=<0 25 25>, angularLimitsMin=<0 -25 -25>, angularTarget=<1 0 0>}
      localJointOffset=<3 0 0>
     }]
    simSpaceSettings {maxAngularAcceleration=25, maxAngularVelocity=5, simSpaceAngularAlpha=1}
   }
   n6 = Anim Dynamics - Bone: tail_01_r!(ComponentPose=@n5, Alpha=1, GravityOverride:Vector="0,0,-100") -> (Pose) {
    angularDampingOverride=1
    angularSpringConstant=7500
    bAngularSpring=true
    bOverrideAngularDamping=true
    bUseGravityOverride=true
    gravityScale=0.35
    simulationSpace=RootRelative
    boundBone {boneName=tail_01_r}
    componentAppliedLinearAccClamp=<1000 1000 1000>
    componentLinearAccScale=<1 1 1>
    componentLinearVelScale=<1 1 1>
    gravityOverride {z=-100}
    physicsBodyDefinitions [{
      boundBone {boneName=tail_01_r}
      boxExtents=<25 35 35>
      constraintSetup {angularLimitsMax=<0 25 25>, angularLimitsMin=<0 -25 -25>, angularTarget=<1 0 0>}
      localJointOffset=<4 0 0>
     }]
    simSpaceSettings {maxAngularAcceleration=25, maxAngularVelocity=5, simSpaceAngularAlpha=1}
   }
   n7 = Anim Dynamics - Bone: tail_02_r!(ComponentPose=@n6, Alpha=1, GravityOverride:Vector="0,0,-60") -> (Pose) {
    angularDampingOverride=1
    angularSpringConstant=7500
    bAngularSpring=true
    bOverrideAngularDamping=true
    bUseGravityOverride=true
    gravityScale=0.35
    simulationSpace=RootRelative
    boundBone {boneName=tail_02_r}
    componentAppliedLinearAccClamp=<1000 1000 1000>
    componentLinearAccScale=<1 1 1>
    componentLinearVelScale=<1 1 1>
    gravityOverride {z=-100}
    physicsBodyDefinitions [{
      boundBone {boneName=tail_02_r}
      boxExtents=<25 35 35>
      constraintSetup {angularLimitsMax=<0 25 25>, angularLimitsMin=<0 -25 -25>, angularTarget=<1 0 0>}
      localJointOffset=<4 0 0>
     }]
    simSpaceSettings {maxAngularAcceleration=25, maxAngularVelocity=5, simSpaceAngularAlpha=1}
   }
   n8 = Anim Dynamics - Bone: tail_03_r!(ComponentPose=@n7, Alpha=1, GravityOverride:Vector="0,0,-60") -> (Pose) {
    angularDampingOverride=1
    angularSpringConstant=7500
    bAngularSpring=true
    bOverrideAngularDamping=true
    bUseGravityOverride=true
    gravityScale=0.35
    simulationSpace=RootRelative
    boundBone {boneName=tail_03_r}
    componentAppliedLinearAccClamp=<1000 1000 1000>
    componentLinearAccScale=<1 1 1>
    componentLinearVelScale=<1 1 1>
    gravityOverride {z=-100}
    physicsBodyDefinitions [{
      boundBone {boneName=tail_03_r}
      boxExtents=<25 35 35>
      constraintSetup {angularLimitsMax=<0 25 25>, angularLimitsMin=<0 -25 -25>, angularTarget=<1 0 0>}
      localJointOffset=<4 0 0>
     }]
    simSpaceSettings {maxAngularAcceleration=25, maxAngularVelocity=5, simSpaceAngularAlpha=1}
   }
   n9 = Anim Dynamics - Bone: tail_04_r!(ComponentPose=@n8, Alpha=1, GravityOverride:Vector="0,0,-60") -> (Pose) {
    angularDampingOverride=1
    angularSpringConstant=7500
    bAngularSpring=true
    bOverrideAngularDamping=true
    bUseGravityOverride=true
    gravityScale=0.35
    simulationSpace=RootRelative
    boundBone {boneName=tail_04_r}
    componentAppliedLinearAccClamp=<1000 1000 1000>
    componentLinearAccScale=<1 1 1>
    componentLinearVelScale=<1 1 1>
    gravityOverride {z=-100}
    physicsBodyDefinitions [{
      boundBone {boneName=tail_04_r}
      boxExtents=<25 35 35>
      constraintSetup {angularLimitsMax=<0 25 25>, angularLimitsMin=<0 -25 -25>, angularTarget=<1 0 0>}
      localJointOffset=<4 0 0>
     }]
    simSpaceSettings {maxAngularAcceleration=25, maxAngularVelocity=5, simSpaceAngularAlpha=1}
   }
   n10 = Anim Dynamics - Bone: tail_05_r!(ComponentPose=@n9, Alpha=1, GravityOverride:Vector="0,0,-60") -> (Pose) {
    angularDampingOverride=1
    angularSpringConstant=7500
    bAngularSpring=true
    bOverrideAngularDamping=true
    bUseGravityOverride=true
    gravityScale=0.35
    simulationSpace=RootRelative
    boundBone {boneName=tail_05_r}
    componentAppliedLinearAccClamp=<1000 1000 1000>
    componentLinearAccScale=<1 1 1>
    componentLinearVelScale=<1 1 1>
    gravityOverride {z=-100}
    physicsBodyDefinitions [{
      boundBone {boneName=tail_05_r}
      boxExtents=<25 35 35>
      constraintSetup {angularLimitsMax=<0 25 25>, angularLimitsMin=<0 -25 -25>, angularTarget=<1 0 0>}
      localJointOffset=<4 0 0>
     }]
    simSpaceSettings {maxAngularAcceleration=25, maxAngularVelocity=5, simSpaceAngularAlpha=1}
   }
   n11 = Anim Dynamics - Bone: hair_top_01!(ComponentPose=@n10, Alpha=1, GravityOverride:Vector="0,0,-100") -> (Pose) {
    angularDampingOverride=1
    angularSpringConstant=5000
    bAngularSpring=true
    bOverrideAngularDamping=true
    bUseGravityOverride=true
    gravityScale=0.35
    simulationSpace=RootRelative
    boundBone {boneName=hair_top_01}
    componentAppliedLinearAccClamp=<1000 1000 1000>
    componentLinearAccScale=<1 1 1>
    componentLinearVelScale=<1 1 1>
    gravityOverride {z=-100}
    physicsBodyDefinitions [{
      boundBone {boneName=hair_top_01}
      boxExtents=<15 25 25>
      constraintSetup {angularLimitsMax=<10 35 35>, angularLimitsMin=<-10 -35 -35>, angularTarget=<1 0 0>}
      localJointOffset=<2.5 0 0>
     }]
    simSpaceSettings {maxAngularAcceleration=36, maxAngularVelocity=6, simSpaceAngularAlpha=1}
   }
   n12 = Anim Dynamics - Bone: hair_mid_01!(ComponentPose=@n11, Alpha=1, GravityOverride:Vector="0,0,-100") -> (Pose) {
    angularDampingOverride=1
    angularSpringConstant=5000
    bAngularSpring=true
    bOverrideAngularDamping=true
    bUseGravityOverride=true
    gravityScale=0.35
    simulationSpace=RootRelative
    boundBone {boneName=hair_mid_01}
    componentAppliedLinearAccClamp=<1000 1000 1000>
    componentLinearAccScale=<1 1 1>
    componentLinearVelScale=<1 1 1>
    gravityOverride {z=-100}
    physicsBodyDefinitions [{
      boundBone {boneName=hair_mid_01}
      boxExtents=<15 25 25>
      constraintSetup {angularLimitsMax=<10 35 35>, angularLimitsMin=<-10 -35 -35>, angularTarget=<1 0 0>}
      localJointOffset=<3 0 0>
     }]
    simSpaceSettings {maxAngularAcceleration=36, maxAngularVelocity=6, simSpaceAngularAlpha=1}
   }
   n13 = Anim Dynamics - Bone: hair_low_01!(ComponentPose=@n12, Alpha=1, GravityOverride:Vector="0,0,-100") -> (Pose) {
    angularDampingOverride=1
    angularSpringConstant=5000
    bAngularSpring=true
    bOverrideAngularDamping=true
    bUseGravityOverride=true
    gravityScale=0.35
    simulationSpace=RootRelative
    boundBone {boneName=hair_low_01}
    componentAppliedLinearAccClamp=<1000 1000 1000>
    componentLinearAccScale=<1 1 1>
    componentLinearVelScale=<1 1 1>
    gravityOverride {z=-100}
    physicsBodyDefinitions [{
      boundBone {boneName=hair_low_01}
      boxExtents=<20 30 30>
      constraintSetup {angularLimitsMax=<10 35 35>, angularLimitsMin=<-10 -35 -35>, angularTarget=<1 0 0>}
      localJointOffset=<2.5 0 0>
     }]
    simSpaceSettings {maxAngularAcceleration=36, maxAngularVelocity=6, simSpaceAngularAlpha=1}
   }
   n14 = Anim Dynamics - Bone: Whiskers_02_l!(ComponentPose=@n13, Alpha=1, GravityOverride:Vector="0,0,-100") -> (Pose) {
    angularDampingOverride=1
    angularSpringConstant=5000
    bAngularSpring=true
    bOverrideAngularDamping=true
    bUseGravityOverride=true
    gravityScale=0.35
    simulationSpace=RootRelative
    boundBone {boneName=Whiskers_02_l}
    componentAppliedLinearAccClamp=<1000 1000 1000>
    componentLinearAccScale=<1 1 1>
    componentLinearVelScale=<1 1 1>
    gravityOverride {z=-100}
    physicsBodyDefinitions [{
      boundBone {boneName=Whiskers_02_l}
      boxExtents=<20 30 30>
      constraintSetup {angularLimitsMax=<10 35 35>, angularLimitsMin=<-10 -35 -35>, angularTarget=<1 0 0>}
      localJointOffset=<2 0 0>
     }]
    simSpaceSettings {maxAngularAcceleration=36, maxAngularVelocity=6, simSpaceAngularAlpha=1}
   }
   n15 = Anim Dynamics - Bone: Whiskers_02_r!(ComponentPose=@n14, Alpha=1, GravityOverride:Vector="0,0,-100") -> (Pose) {
    angularDampingOverride=1
    angularSpringConstant=5000
    bAngularSpring=true
    bOverrideAngularDamping=true
    bUseGravityOverride=true
    gravityScale=0.35
    simulationSpace=RootRelative
    boundBone {boneName=Whiskers_02_r}
    componentAppliedLinearAccClamp=<1000 1000 1000>
    componentLinearAccScale=<1 1 1>
    componentLinearVelScale=<1 1 1>
    gravityOverride {z=-100}
    physicsBodyDefinitions [{
      boundBone {boneName=Whiskers_02_r}
      boxExtents=<20 30 30>
      constraintSetup {angularLimitsMax=<10 35 35>, angularLimitsMin=<-10 -35 -35>, angularTarget=<1 0 0>}
      localJointOffset=<-2 0 0>
     }]
    simSpaceSettings {maxAngularAcceleration=36, maxAngularVelocity=6, simSpaceAngularAlpha=1}
   }
   Save cached pose 'DynamicsCache'(Pose=Component To Local(ComponentPose=@n15)) {cacheName=DynamicsCache}
   n16 = "Transform (Modify) Bone - Bone: tail_01"!(ComponentPose=Local To Component(LocalPose=Use cached pose 'CRCache_Input'()), Alpha=1, Rotation="25,10,-15") -> (Pose) {
    rotationMode=BMM_Additive
    rotationSpace=BCS_BoneSpace
    alphaScaleBiasClamp {bInterpResult=true}
    boneToModify {boneName=tail_01}
    rotation=<rot 25 10 -15>
   }
   n17 = "Transform (Modify) Bone - Bone: tail_03"!(ComponentPose=@n16, Alpha=1, Rotation="10,5,0") -> (Pose) {
    rotationMode=BMM_Additive
    rotationSpace=BCS_BoneSpace
    alphaScaleBiasClamp {bInterpResult=true}
    boneToModify {boneName=tail_03}
    rotation {pitch=10, yaw=5}
   }
   n18 = "Transform (Modify) Bone - Bone: tail_01_r"(ComponentPose=@n17, Alpha=1, Rotation="-25,25,0") {
    rotationMode=BMM_Additive
    rotationSpace=BCS_BoneSpace
    alphaScaleBiasClamp {bInterpResult=true, interpSpeedDecreasing=8}
    boneToModify {boneName=tail_01_r}
    rotation {pitch=-25, yaw=25}
   }
   n19 = "Transform (Modify) Bone - Bone: tail_03_r"!(ComponentPose=@n18, Alpha=1, Rotation="-10,15,-15") -> (Pose) {
    rotationMode=BMM_Additive
    rotationSpace=BCS_BoneSpace
    alphaScaleBiasClamp {bInterpResult=true, interpSpeedDecreasing=8}
    boneToModify {boneName=tail_03_r}
    rotation=<rot -10 15 -15>
   }
   n20 = "Transform (Modify) Bone - Bone: tail_01"(ComponentPose=@n19, Alpha=1, Rotation="20,-2.5,0") {
    rotationMode=BMM_Additive
    rotationSpace=BCS_BoneSpace
    alphaScaleBiasClamp {bInterpResult=true}
    boneToModify {boneName=tail_01}
    rotation {pitch=20, yaw=-2.5}
   }
   Save cached pose 'preDynamicsCache'(Pose=Blend Poses by bool(BlendPose_0=Use cached pose 'CRCache_Input'(), BlendPose_1=Component To Local(ComponentPose=@n20))) {cacheName=preDynamicsCache}
  ]}
 EventGraph {nodes [
   Event BlueprintUpdateAnimation()
   > Set R_Tail_Raiser(R_Tail_Raiser:float=GetCurveValue(CurveName=r_tail_raiser)) -> (Output_Get:float)
   > Set Tail_Running(Tail_Running:float=GetCurveValue(CurveName=tail_running)) -> (Output_Get:float)
   n1 = Get DynamicsAlpha() -> (DynamicsAlpha:float)
   n2 = Get Tail_Running() -> (Tail_Running:float)
   > Set L_Tail_Lower_Alpha(L_Tail_Lower_Alpha:double=float * float(A=float - float(A=1, B=@n2), B=float - float(A=1, B=@n1)) {is=title K2Node_PromotableOperator}) -> (Output_Get:double)
   Event BlueprintInitializeAnimation()
   n3 = GetOwningActor()
   n4 = GetOwningActor()
   > Set isAnimatingGameplay(isAnimatingGameplay:bool=OR Boolean(
     A=IsA!(Object:Object Reference=@n4, ObjectClass:Object Class Reference=/CosmeticCompanions/Assets/Shared/Blueprint/BP_Companion.BP_Companion_C) -> (ReturnValue:bool)
     B=IsA(Object=@n3, ObjectClass=/Script/CosmeticCompanionsRuntime.FortCompanionMannequin)
    )) -> (Output_Get:bool)
  ]}
]
variables [
 isAnimatingGameplay {type=boolean}
 L_Tail_Lower_Alpha {type=number}
 Loco_Sprint {type=number}
 PoseAsset {type=PoseAsset, default=/CosmeticCompanions/Automation_Template/companion_pose_asset_template}
 R_Tail_Raiser {type=number}
 Tail_Running {type=number}
]
